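# Physics asset for the male medium AncientLeer body cosmetic, set up to simulate its jacket.
# Pelvis, spine and thigh capsules form an angle-locked body core, with a Bomb capsule locked to the right thigh.
# Seven jacket chains hang from base bodies locked to spine_03, using capsules and boxes with mostly limited swing and twist.
# Free-angle constraints link neighboring chain ends from right front around the back to left front, tying the hem together.
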PhysicsAsset /BRCosmetics/Characters/Player/Male/Medium/Bodies/M_MED_AncientLeer/Meshes/M_MED_AncientLeer_Physics
bodies {
 Bomb [{shapeType=Capsule, length=5.0291, radius=5.7748, shapeName=Bomb_capsule, center=<-0.7381 -0.6381 1.3661>, rotation=<rot -1.0588 -90 -0.9756>}]
 base_Jacket_01_L [{shapeType=Capsule, length=1, radius=1}]
 base_Jacket_01_Lback [{shapeType=Capsule, length=1, radius=1}]
 base_Jacket_01_Lfront [{shapeType=Capsule, length=1, radius=1}]
 base_Jacket_01_R [{shapeType=Capsule, length=1, radius=1}]
 base_Jacket_01_Rback [{shapeType=Capsule, length=1, radius=1}]
 base_Jacket_01_Rfront [{shapeType=Capsule, length=1, radius=1}]
 base_Jacket_01_mid [{shapeType=Capsule, length=1, radius=1}]
 dyn_Jacket_01_L [{shapeType=Capsule, length=15.0498, radius=5.4577, center=<5.8031 0 3.5>, rotation=<rot 0 0 90>}]
 dyn_Jacket_01_Lback [{shapeType=Capsule, length=16.5585, radius=3.949, center=<1 0 3.5>, rotation=<rot 0 0 90>}]
 dyn_Jacket_01_Lfront [{shapeType=Capsule, length=15.0498, radius=5.4577, center=<7.4972 0 3.5>, rotation=<rot 0 90 90>}]
 dyn_Jacket_01_R [{shapeType=Capsule, length=15.0498, radius=5.4577, center=<5.8031 0 3.5>, rotation=<rot 0 0 90>}]
 dyn_Jacket_01_Rback [{shapeType=Capsule, length=16.5585, radius=3.949, center=<1 0 3.5>, rotation=<rot 0 0 90>}]
 dyn_Jacket_01_Rfront [{shapeType=Capsule, length=16.2974, radius=4.2101, center=<7.4972 0 3.5>, rotation=<rot 0 89.9999 100.1058>}]
 dyn_Jacket_01_mid [{shapeType=Capsule, length=15.6586, radius=4.8489, center=<1 0 3.5>, rotation=<rot 0 0 90>}]
 dyn_Jacket_03_L [{shapeType=Box, extentX=10.9791, extentY=13.2549, extentZ=3.854, center=<5.4499 0 0>}]
 dyn_Jacket_03_Lback [{shapeType=Capsule, length=9, radius=6, center=<1 0 3>, rotation=<rot 0 0 90>}]
 dyn_Jacket_03_Lfront [{shapeType=Box, extentX=11.1539, extentY=11.9789, extentZ=3.7723, center=<5.1695 -3 0>}]
 dyn_Jacket_03_R [{shapeType=Box, extentX=12.1679, extentY=10.9064, extentZ=3.151, center=<4.1167 0 0>, rotation=<rot 0 -0.0001 -2.5858>}]
 dyn_Jacket_03_Rback [{shapeType=Capsule, length=9, radius=6, center=<1 0 3>, rotation=<rot 0 0 90>}]
 dyn_Jacket_03_Rfront [{shapeType=Box, extentX=10.4389, extentY=13.1915, extentZ=2.7993, center=<4.7884 5 -2>, rotation=<rot 0.0001 -0.0001 28.9209>}]
 dyn_Jacket_03_mid [{shapeType=Capsule, length=8.7775, radius=4.7932, center=<2.4751 0 0>, rotation=<rot 0 0 90>}]
 dyn_Jacket_04_L [{shapeType=Box, extentX=13.8937, extentY=14.9385, extentZ=3.9573, center=<7.047 0 0>}]
 dyn_Jacket_04_Lback [{shapeType=Capsule, length=9, radius=6, center=<1 0 3>, rotation=<rot 0 0 90>}]
 dyn_Jacket_04_Lfront [{shapeType=Box, extentX=9.103, extentY=13.7095, extentZ=4.2048, center=<3.7 -4 0>}]
 dyn_Jacket_04_R [{shapeType=Box, extentX=9.1842, extentY=13.596, extentZ=2.7923, center=<4.6737 0 0>}]
 dyn_Jacket_04_Rback [{shapeType=Capsule, length=8.8, radius=6.2, center=<1 0 3>, rotation=<rot 0 0 90>}]
 dyn_Jacket_04_Rfront [{shapeType=Box, extentX=12.8815, extentY=14.992, extentZ=4.3736, center=<3.2385 6 -1.5>, rotation=<rot 0 0 27.1875>}]
 dyn_Jacket_04_mid [{shapeType=Capsule, length=7.9775, radius=5.5932, center=<2.4751 0 0>, rotation=<rot 0 0 90>}]
 dyn_Jacket_05_Lback [{shapeType=Capsule, length=8.6, radius=6.4, center=<1 0 3>, rotation=<rot 0 0 90>}]
 dyn_Jacket_05_Lfront [{shapeType=Box, extentX=12.6306, extentY=16.2944, extentZ=3.9704, center=<4.7655 -5.5 1>, rotation=<rot 14.096 0 -2.8126>}]
 dyn_Jacket_05_Rback [{shapeType=Capsule, length=8.2, radius=6.8, center=<1 0 3>, rotation=<rot 0 0 90>}]
 dyn_Jacket_05_Rfront [{shapeType=Box, extentX=11.7416, extentY=12.8882, extentZ=4.4207, center=<5.9094 5.9901 -1.0339>, rotation=<rot -1.1873 -0.346 32.1545>}]
 pelvis [{shapeType=Capsule, length=8.491, radius=10.4005, center=<-2.506 -0.2228 0>}]
 spine_01 [{shapeType=Capsule, length=6.8896, radius=10.4005}]
 spine_02 [{shapeType=Capsule, length=3.3073, radius=10.4005}]
 spine_03 [{shapeType=Capsule, length=7.2209, radius=8.1921}]
 thigh_twist_01_l [{shapeType=Capsule, length=27.6238, radius=9.3625, center=<-6.15 -0.9298 3.3817>, rotation=<rot 85.7868 6.5813 3.0905>}]
 thigh_twist_01_r [{shapeType=Capsule, length=27.4061, radius=8.5478, center=<6.2102 0.9368 -1.2684>, rotation=<rot 89.1571 6.5815 -179.9996>}]
}
constraints [
 {
  bone1Name=spine_02
  bone2Name=spine_01
  swing1LimitDegrees=45
  swing1Motion=Locked
  swing2LimitDegrees=45
  swing2Motion=Locked
  twistLimitDegrees=45
  twistMotion=Locked
 }
 {
  bone1Name=spine_03
  bone2Name=spine_02
  swing1LimitDegrees=45
  swing1Motion=Locked
  swing2LimitDegrees=45
  swing2Motion=Locked
  twistLimitDegrees=45
  twistMotion=Locked
 }
 {
  bone1Name=spine_01
  bone2Name=pelvis
  swing1LimitDegrees=45
  swing1Motion=Locked
  swing2LimitDegrees=45
  swing2Motion=Locked
  twistLimitDegrees=45
  twistMotion=Locked
 }
 {
  bone1Name=base_Jacket_01_Lback
  bone2Name=spine_03
  swing1LimitDegrees=5
  swing1Motion=Locked
  swing2LimitDegrees=20
  swing2Motion=Locked
  twistLimitDegrees=2
  twistMotion=Locked
 }
 {
  bone1Name=base_Jacket_01_Lfront
  bone2Name=spine_03
  swing1LimitDegrees=5
  swing1Motion=Locked
  swing2LimitDegrees=20
  swing2Motion=Locked
  twistLimitDegrees=2
  twistMotion=Locked
 }
 {
  bone1Name=base_Jacket_01_R
  bone2Name=spine_03
  swing1LimitDegrees=5
  swing1Motion=Locked
  swing2LimitDegrees=20
  swing2Motion=Locked
  twistLimitDegrees=2
  twistMotion=Locked
 }
 {
  bone1Name=base_Jacket_01_Rback
  bone2Name=spine_03
  swing1LimitDegrees=5
  swing1Motion=Locked
  swing2LimitDegrees=20
  swing2Motion=Locked
  twistLimitDegrees=2
  twistMotion=Locked
 }
 {
  bone1Name=base_Jacket_01_Rfront
  bone2Name=spine_03
  swing1LimitDegrees=5
  swing1Motion=Locked
  swing2LimitDegrees=20
  swing2Motion=Locked
  twistLimitDegrees=2
  twistMotion=Locked
 }
 {
  bone1Name=base_Jacket_01_mid
  bone2Name=spine_03
  swing1LimitDegrees=5
  swing1Motion=Locked
  swing2LimitDegrees=20
  swing2Motion=Locked
  twistLimitDegrees=2
  twistMotion=Locked
 }
 {
  bone1Name=base_Jacket_01_L
  bone2Name=spine_03
  swing1LimitDegrees=5
  swing1Motion=Locked
  swing2LimitDegrees=20
  swing2Motion=Locked
  twistLimitDegrees=2
  twistMotion=Locked
 }
 {
  bone1Name=thigh_twist_01_r
  bone2Name=pelvis
  swing1LimitDegrees=45
  swing1Motion=Locked
  swing2LimitDegrees=45
  swing2Motion=Locked
  twistLimitDegrees=45
  twistMotion=Locked
 }
 {
  bone1Name=thigh_twist_01_l
  bone2Name=pelvis
  swing1LimitDegrees=45
  swing1Motion=Locked
  swing2LimitDegrees=45
  swing2Motion=Locked
  twistLimitDegrees=45
  twistMotion=Locked
 }
 {bone1Name=dyn_Jacket_01_L, bone2Name=base_Jacket_01_L, swing1LimitDegrees=10, swing2LimitDegrees=25, twistLimitDegrees=5}
 {bone1Name=dyn_Jacket_01_Lback, bone2Name=base_Jacket_01_Lback, swing1LimitDegrees=10, swing2LimitDegrees=25, twistLimitDegrees=5}
 {bone1Name=dyn_Jacket_01_Lfront, bone2Name=base_Jacket_01_Lfront, swing1LimitDegrees=10, swing2LimitDegrees=25, twistLimitDegrees=5}
 {bone1Name=dyn_Jacket_01_R, bone2Name=base_Jacket_01_R, swing1LimitDegrees=10, swing2LimitDegrees=25, twistLimitDegrees=5}
 {bone1Name=dyn_Jacket_01_Rback, bone2Name=base_Jacket_01_Rback, swing1LimitDegrees=10, swing2LimitDegrees=25, twistLimitDegrees=5}
 {bone1Name=dyn_Jacket_01_Rfront, bone2Name=base_Jacket_01_Rfront, swing1LimitDegrees=14, swing2LimitDegrees=25, twistLimitDegrees=5}
 {bone1Name=dyn_Jacket_01_mid, bone2Name=base_Jacket_01_mid, swing1LimitDegrees=10, swing2LimitDegrees=25, twistLimitDegrees=5}
 {bone1Name=dyn_Jacket_04_L, bone2Name=dyn_Jacket_03_L, swing1LimitDegrees=10, swing2LimitDegrees=25, twistLimitDegrees=5}
 {bone1Name=dyn_Jacket_03_L, bone2Name=dyn_Jacket_01_L, swing1LimitDegrees=10, swing2LimitDegrees=25, twistLimitDegrees=5}
 {bone1Name=dyn_Jacket_05_Lfront, bone2Name=dyn_Jacket_04_Lfront, swing1LimitDegrees=5, swing2LimitDegrees=25, twistLimitDegrees=5}
 {bone1Name=dyn_Jacket_04_Lfront, bone2Name=dyn_Jacket_03_Lfront, swing1LimitDegrees=10, swing2LimitDegrees=25, twistLimitDegrees=5}
 {bone1Name=dyn_Jacket_03_Lfront, bone2Name=dyn_Jacket_01_Lfront, swing1LimitDegrees=10, swing2LimitDegrees=25, twistLimitDegrees=5}
 {bone1Name=dyn_Jacket_04_R, bone2Name=dyn_Jacket_03_R, swing1LimitDegrees=10, swing2LimitDegrees=25, twistLimitDegrees=5}
 {bone1Name=dyn_Jacket_03_R, bone2Name=dyn_Jacket_01_R, swing1LimitDegrees=10, swing2LimitDegrees=25, twistLimitDegrees=5}
 {bone1Name=dyn_Jacket_05_Rfront, bone2Name=dyn_Jacket_04_Rfront, swing1LimitDegrees=10, swing2LimitDegrees=12, twistLimitDegrees=7}
 {bone1Name=dyn_Jacket_04_Rfront, bone2Name=dyn_Jacket_03_Rfront, swing1LimitDegrees=10, swing2LimitDegrees=25, twistLimitDegrees=7}
 {bone1Name=dyn_Jacket_03_Rfront, bone2Name=dyn_Jacket_01_Rfront, swing1LimitDegrees=10, swing2LimitDegrees=25, twistLimitDegrees=7}
 {bone1Name=dyn_Jacket_04_mid, bone2Name=dyn_Jacket_03_mid, swing1LimitDegrees=10, swing2LimitDegrees=25, twistLimitDegrees=5}
 {bone1Name=dyn_Jacket_03_mid, bone2Name=dyn_Jacket_01_mid, swing1LimitDegrees=10, swing2LimitDegrees=25, twistLimitDegrees=5}
 {bone1Name=dyn_Jacket_03_Lback, bone2Name=dyn_Jacket_01_Lback, swing1LimitDegrees=10, swing2LimitDegrees=25, twistLimitDegrees=5}
 {bone1Name=dyn_Jacket_04_Lback, bone2Name=dyn_Jacket_03_Lback, swing1LimitDegrees=10, swing2LimitDegrees=25, twistLimitDegrees=5}
 {bone1Name=dyn_Jacket_05_Lback, bone2Name=dyn_Jacket_04_Lback, swing1LimitDegrees=10, swing2LimitDegrees=25, twistLimitDegrees=5}
 {bone1Name=dyn_Jacket_03_Rback, bone2Name=dyn_Jacket_01_Rback, swing1LimitDegrees=10, swing2LimitDegrees=25, twistLimitDegrees=5}
 {bone1Name=dyn_Jacket_04_Rback, bone2Name=dyn_Jacket_03_Rback, swing1LimitDegrees=10, swing2LimitDegrees=25, twistLimitDegrees=5}
 {bone1Name=dyn_Jacket_05_Rback, bone2Name=dyn_Jacket_04_Rback, swing1LimitDegrees=10, swing2LimitDegrees=25, twistLimitDegrees=5}
 {
  bone1Name=dyn_Jacket_05_Rfront
  bone2Name=dyn_Jacket_04_R
  swing1LimitDegrees=45
  swing1Motion=Free
  swing2LimitDegrees=45
  swing2Motion=Free
  twistLimitDegrees=45
  twistMotion=Free
 }
 {
  bone1Name=dyn_Jacket_04_R
  bone2Name=dyn_Jacket_05_Rback
  swing1LimitDegrees=45
  swing1Motion=Free
  swing2LimitDegrees=45
  swing2Motion=Free
  twistLimitDegrees=45
  twistMotion=Free
 }
 {
  bone1Name=dyn_Jacket_05_Rback
  bone2Name=dyn_Jacket_04_mid
  swing1LimitDegrees=45
  swing1Motion=Free
  swing2LimitDegrees=45
  swing2Motion=Free
  twistLimitDegrees=45
  twistMotion=Free
 }
 {
  bone1Name=dyn_Jacket_05_Lfront
  bone2Name=dyn_Jacket_04_L
  swing1LimitDegrees=45
  swing1Motion=Free
  swing2LimitDegrees=45
  swing2Motion=Free
  twistLimitDegrees=45
  twistMotion=Free
 }
 {
  bone1Name=dyn_Jacket_04_L
  bone2Name=dyn_Jacket_05_Lback
  swing1LimitDegrees=45
  swing1Motion=Free
  swing2LimitDegrees=45
  swing2Motion=Free
  twistLimitDegrees=45
  twistMotion=Free
 }
 {
  bone1Name=dyn_Jacket_05_Lback
  bone2Name=dyn_Jacket_04_mid
  swing1LimitDegrees=45
  swing1Motion=Free
  swing2LimitDegrees=45
  swing2Motion=Free
  twistLimitDegrees=45
  twistMotion=Free
 }
 {
  bone1Name=Bomb
  bone2Name=thigh_twist_01_r
  swing1LimitDegrees=45
  swing1Motion=Locked
  swing2LimitDegrees=45
  swing2Motion=Locked
  twistLimitDegrees=45
  twistMotion=Locked
 }
]
