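# Physics asset of cape colliders for the AncientLeer medium male player cosmetic body.
# Capsule bodies cover the pelvis, upper spine, arms, hands, legs and feet, plus small capsules on a subset of the dynamic jacket back and mid bones.
# It has no constraints, so it only defines collision shapes on bones rather than a jointed ragdoll, presumably for cape simulation to collide against.
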PhysicsAsset /BRCosmetics/Characters/Player/Male/Medium/Bodies/M_MED_AncientLeer/Meshes/M_MED_AncientLeer_CapeColliders
bodies {
 calf_l [{shapeType=Capsule, length=32.5143, radius=15.6454, center=<-23.0006 -2.2427 0>, rotation=<rot 90 0 0>}]
 calf_r [{shapeType=Capsule, length=32.5143, radius=15.6454, center=<23.0006 2.2427 0>, rotation=<rot 90 0 180>}]
 dyn_Jacket_01_mid [{shapeType=Capsule, length=6, radius=5.4, rotation=<rot 0 0 90>}]
 dyn_Jacket_03_Lback [{shapeType=Capsule, length=6, radius=5.4, rotation=<rot 0 0 90>}]
 dyn_Jacket_03_Rback [{shapeType=Capsule, length=6, radius=5.4, rotation=<rot 0 0 90>}]
 dyn_Jacket_03_mid [{shapeType=Capsule, length=5.4, radius=6, rotation=<rot 0 0 90>}]
 dyn_Jacket_04_Lback [{shapeType=Capsule, length=5.6, radius=5.8, rotation=<rot 0 0 90>}]
 dyn_Jacket_04_Rback [{shapeType=Capsule, length=5.6, radius=5.8, rotation=<rot 0 0 90>}]
 dyn_Jacket_04_mid [{shapeType=Capsule, length=5.4, radius=6, rotation=<rot 0 0 90>}]
 dyn_Jacket_05_Lback [{shapeType=Capsule, length=4.3, radius=6.5, rotation=<rot 0 0 90>}]
 dyn_Jacket_05_Rback [{shapeType=Capsule, length=4.3, radius=6.5, rotation=<rot 0 0 90>}]
 foot_l [{shapeType=Capsule, length=10.5461, radius=14.1734, center=<-4.1955 6.0815 0>, rotation=<rot 0 14.9998 90>}]
 foot_r [{shapeType=Capsule, length=10.5461, radius=14.1734, center=<4.1955 -6.0815 0>, rotation=<rot 0 14.9998 -90>}]
 hand_l [{shapeType=Capsule, length=9.5015, radius=6.7933, center=<6.2703 0 0>, rotation=<rot 90 0 0>}]
 hand_r [{shapeType=Capsule, length=9.5015, radius=6.7933, center=<-6.2703 0 0>, rotation=<rot 90 0 180>}]
 lowerarm_l [{shapeType=Capsule, length=14.3387, radius=7.5574, center=<8.9852 0 0.2267>, rotation=<rot 90 0 0>}]
 lowerarm_r [{shapeType=Capsule, length=14.3387, radius=7.5574, center=<-8.9852 0 -0.2267>, rotation=<rot 90 0 180>}]
 pelvis [{shapeType=Capsule, length=5.0241, radius=17.7874, center=<-13.7107 -0.9262 0>}]
 spine_02 [{shapeType=Capsule, length=0.7, radius=18.0784, center=<-2.6142 0 0>}]
 spine_03 [{shapeType=Capsule, length=3.4972, radius=16.2226}]
 spine_04 [{shapeType=Capsule, length=2.5255, radius=17.3781, center=<0 -0.0477 0>}]
 spine_05 [{shapeType=Capsule, length=10.9226, radius=18.9992, center=<0 0.7708 0>}]
 thigh_l [{shapeType=Capsule, length=15.4697, radius=14.5854, center=<-18.4217 -1.038 0>, rotation=<rot 90 180 2.664>}]
 thigh_r [{shapeType=Capsule, length=16.4633, radius=13.5918, center=<18.4217 1.038 0>, rotation=<rot 90 180 -177.336>}]
 upperarm_l [{shapeType=Capsule, length=17.1712, radius=7.7501, center=<10.0455 0 0>, rotation=<rot 90 0 0>}]
 upperarm_r [{shapeType=Capsule, length=17.1712, radius=7.7501, center=<-10.0455 0 0>, rotation=<rot 90 0 180>}]
}
constraints []
