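# Physics asset for the AncientLeer male medium character's FaceAcc part, setting up physics for leaf pieces attached at both upper arms.
# Each side chains two leaf capsules through a small base-leaf capsule, with tight swing and twist limits between leaf bodies.
# The bases lock angularly to the upper arms, which lock to spine_05.
# Upper-arm and spine_03 to spine_05 capsules form the body volume, with free angular joints along the spine.
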PhysicsAsset /BRCosmetics/Characters/Player/Male/Medium/Bodies/M_MED_AncientLeer/Meshes/Parts/M_MED_AncientLeer_FaceAcc_Physics
bodies {
 base_leaves_01_l [{shapeType=Capsule, length=1, radius=1}]
 base_leaves_02_r [{shapeType=Capsule, length=1, radius=1}]
 leaves_01_l [{shapeType=Capsule, length=4.9579, radius=3.5431, center=<1 0 0>, rotation=<rot 90 0 0>}]
 leaves_01_r [{shapeType=Capsule, length=5.156, radius=3.345, center=<1 0 0>, rotation=<rot 90 0 0>}]
 leaves_02_l [{shapeType=Capsule, length=3.0439, radius=5.4571, center=<1 0 0>, rotation=<rot 90 0 0>}]
 leaves_02_r [{shapeType=Capsule, length=5.156, radius=3.345, center=<1 0 0>, rotation=<rot 90 0 0>}]
 spine_03 [
  {shapeType=Capsule, length=11.7748, radius=13.5488}
  {shapeType=Capsule, length=1.7536, radius=15, shapeName=spine_03_capsule, center=<0 -5.2229 -3.3221>}
 ]
 spine_04 [{shapeType=Capsule, length=1.1454, radius=17.0443, shapeName=spine_04_capsule, center=<1.8557 7.5187 0>, rotation=<rot 0 -90 0>}]
 spine_05 [{
   shapeType=Capsule
   length=6.2408
   radius=7.9387
   shapeName=spine_05_capsule
   center=<-0.5088 -0.191 0.2601>
   rotation=<rot -3.0173 -90 2.0216>
  }]
 upperarm_l [{shapeType=Capsule, length=8.611, radius=5.4761, shapeName=upperarm_l_capsule, center=<9 0 0>, rotation=<rot 90 180 0>}]
 upperarm_r [{shapeType=Capsule, length=8.611, radius=5.4761, shapeName=upperarm_r_capsule, center=<-9 0 0>, rotation=<rot 90 0 0>}]
}
constraints [
 {bone1Name=leaves_01_r, bone2Name=leaves_02_r, swing1LimitDegrees=5, swing2LimitDegrees=10, twistLimitDegrees=1}
 {bone1Name=leaves_02_r, bone2Name=base_leaves_02_r, swing1LimitDegrees=10, swing2LimitDegrees=20, twistLimitDegrees=4}
 {
  bone1Name=base_leaves_02_r
  bone2Name=upperarm_r
  swing1LimitDegrees=45
  swing1Motion=Locked
  swing2LimitDegrees=45
  swing2Motion=Locked
  twistLimitDegrees=45
  twistMotion=Locked
 }
 {bone1Name=leaves_01_l, bone2Name=base_leaves_01_l, swing1LimitDegrees=10, swing2LimitDegrees=20, twistLimitDegrees=1}
 {bone1Name=leaves_02_l, bone2Name=leaves_01_l, swing1LimitDegrees=5, swing2LimitDegrees=10, twistLimitDegrees=1}
 {
  bone1Name=base_leaves_01_l
  bone2Name=upperarm_l
  swing1LimitDegrees=45
  swing1Motion=Locked
  swing2LimitDegrees=45
  swing2Motion=Locked
  twistLimitDegrees=45
  twistMotion=Locked
 }
 {
  bone1Name=upperarm_r
  bone2Name=spine_05
  swing1LimitDegrees=45
  swing1Motion=Locked
  swing2LimitDegrees=45
  swing2Motion=Locked
  twistLimitDegrees=45
  twistMotion=Locked
 }
 {
  bone1Name=spine_05
  bone2Name=spine_04
  swing1LimitDegrees=45
  swing1Motion=Free
  swing2LimitDegrees=45
  swing2Motion=Free
  twistLimitDegrees=45
  twistMotion=Free
 }
 {
  bone1Name=upperarm_l
  bone2Name=spine_05
  swing1LimitDegrees=45
  swing1Motion=Locked
  swing2LimitDegrees=45
  swing2Motion=Locked
  twistLimitDegrees=45
  twistMotion=Locked
 }
 {
  bone1Name=spine_04
  bone2Name=spine_03
  swing1LimitDegrees=45
  swing1Motion=Free
  swing2LimitDegrees=45
  swing2Motion=Free
  twistLimitDegrees=45
  twistMotion=Free
 }
]
